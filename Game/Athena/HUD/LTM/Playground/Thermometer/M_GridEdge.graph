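# Unlit, additive, two-sided material for the Playground Thermometer HUD, drawing a glowing edge where its mesh nears scene geometry behind it.
# Two depth-distance falloffs, one tight and one wide and faint, form the intersection glow.
# A procedural stripe mask from one UV axis, tiled by rounded object X scale, boosts the wide falloff.
# The sum is tinted by an HDR edge color; a heatmap boxOpacity collection parameter has no visible connection.
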Material /Game/Athena/HUD/LTM/Playground/Thermometer/M_GridEdge
graph {
 properties {
  bCastRayTracedShadows=false
  blendMode=BLEND_Additive
  pixelDepthOffsetMode=PDOM_Legacy
  shadingModel=MSM_Unlit
  twoSided=true
  usedShadingModels=Unlit
 }
 nodes [
  CollectionParameter() {collection=&/MPC_heatmap, parameterName=boxOpacity}
  n1 = MaterialFunctionCall!(Distance Bias, Distance Divisor=Constant() {r=5}, Scene Depth, Pixel Depth) {materialFunction=/Game/Effects/Fort_Effects/Materials/_Functions/MF_DistanceBehindTranslucency_01}
  n2 = Add(
   A=Power(Base=@n1.Normalized Inverted, Exp=ScalarParameter() {defaultValue=6, parameterName=Intersection Edge Exponent})
   B=Multiply(A=@n1.Normalized Inverted)
  )
  n3 = MaterialFunctionCall(Distance Divisor=ScalarParameter() {defaultValue=75, parameterName=Intersection Edge Divisor}) {materialFunction=&MF_DistanceBehindTranslucency_01}
  n4 = Power(Base=@n3.Normalized Inverted, Exp=ScalarParameter() {defaultValue=6, parameterName=Intersection Edge Exponent})
  n5 = Add(A=Multiply(A=Saturate(None=@n4)) {constB=0.02}, B=Multiply(A=@n3.Normalized Inverted) {constB=0.02})
  n6 = Max(A=MaterialFunctionCall!(In=MaterialFunctionCall() {materialFunction=/Engine/Functions/Engine_MaterialFunctions02/WorldPositionOffset/ObjectScale}.Scale X) {materialFunction=/Engine/Functions/Engine_MaterialFunctions02/Utility/Round}.Result)
  n7 = ScalarParameter() {defaultValue=2, parameterName=Tile UV Scale}
  n8 = Multiply(A=Multiply(A=Divide(A=@n6, B=@n7), B=TextureCoordinate()) {constA=1}) {constB=250}
  n9 = Add(A=ComponentMask(None=Frac(None=@n8)) {g=true}, B=ScalarParameter() {defaultValue=-0.245, parameterName=Line Ceil Bias})
  n10 = Saturate(None=Ceil(None=Clamp(None=@n9)))
  Multiply(
   A=Add(A=Add(A=@n2, B=@n5), B=Multiply(A=@n5, B=@n10))
   B=VectorParameter() {parameterName=Intersection Edge Color, defaultValue=<rgba 0.1875 4.8781 15 1>}.RGB
  )
 ]
}
