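# Unlit material, named for the heatmap UI in the Playground thermometer HUD, that samples a heatmap texture parameter.
# UVs rotate a quarter turn, then scale by HeatmapMaskScale from the MPC_heatmap parameter collection to sample Heatmap_code.
# A separate branch quantizes the rotated UVs into grid steps with divide and ceil but does not feed the sample.
# The graph shows no connections to material outputs.
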Material /Game/Athena/HUD/LTM/Playground/Thermometer/M_HeatmapUI
graph {
 properties {pixelDepthOffsetMode=PDOM_Legacy, shadingModel=MSM_Unlit, usedShadingModels=Unlit}
 nodes [
  n1 = MaterialFunctionCall(UVs=TextureCoordinate(), "Rotation Angle (0-1)"=Constant() {r=0.25}) {materialFunction=&CustomRotator}
  n2 = Divide(B=Constant() {r=32}) {constB=32}
  Divide(A=Ceil(None=Divide(A=@n1.Rotated Values, B=@n2)), B=Constant() {r=33}) {constB=32}
  TextureSample(
   UVs=Multiply(
    A=@n1.Rotated Values
    B=CollectionParameter() {collection=&/MPC_heatmap, parameterName=HeatmapMaskScale}
   )
   Tex=TextureObjectParameter() {parameterName=Heatmap_code}
  )
 ]
}
