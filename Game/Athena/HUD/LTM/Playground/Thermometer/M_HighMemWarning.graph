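# Translucent UI material for the high-memory warning in the Playground LTM thermometer HUD, showing an animated exclamation texture.
# A Panner scrolls the texture parameter's UVs slowly at a constant rate.
# A periodic time-driven sine, remapped by bias-scale, drives a lerp that multiplies the sampled RGB, so the warning pulses between dim and full brightness without vanishing.
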Material /Game/Athena/HUD/LTM/Playground/Thermometer/M_HighMemWarning
graph {
 properties {blendMode=BLEND_Translucent, materialDomain=MD_UI, pixelDepthOffsetMode=PDOM_Legacy, usedShadingModels=Default Lit}
 nodes [
  n1 = TextureCoordinate()
  n2 = Constant() {r=-0.1}
  n3 = Sine(None=Time() {bOverride_Period=true, period=3})
  Multiply(
   A=TextureSample(
    UVs=Panner(Coordinate=@n1, Speed=@n2) {speedX=1}
    Tex=TextureObjectParameter() {parameterName=Texture, texture=&/T_Exclamaition}
   ) {texture=&T_Exclamaition}.RGB
   B=LinearInterpolate(Alpha=ConstantBiasScale(None=@n3)) {constA=0.25}
  )
 ]
}
