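# Material function that supplies UVs and masks for the Playground thermometer minimap, a grid heatmap in the HUD.
# It remaps MPC_heatmap player and aim positions into the volume bounds, snaps them to cell or three-cell group centers, and rotates them a quarter turn.
# Offsets of three cells yield the A1–C3 neighbors around the player.
# The cell input samples group, warning and footprint masks, plus a time-pulsed grid effect.
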MaterialFunction /Game/Athena/HUD/LTM/Playground/Thermometer/MF_minimapUIUV
graph {nodes [
  n1 = MaterialFunctionCall!(
   Input=CollectionParameter() {collection=&/MPC_heatmap, parameterName=PlayerX}
   Input Low=CollectionParameter() {collection=&MPC_heatmap, parameterName=volumeSizeMInX}
   Input High=CollectionParameter() {collection=&MPC_heatmap, parameterName=volumeSizeMaxX}
   Target Low=Constant()
   Target Height=Constant() {r=1}
  ) {materialFunction=&RemapValueRange}
  n2 = CollectionParameter() {collection=&MPC_heatmap, parameterName=Gridsize}
  n3 = Floor(None=Multiply(A=@n1.Result, B=@n2) {constB=32})
  n4 = Multiply(A=@n2) {constB=2}
  n5 = MaterialFunctionCall(
   Input=CollectionParameter() {collection=&MPC_heatmap, parameterName=PlayerY}
   Input Low=CollectionParameter() {collection=&MPC_heatmap, parameterName=volumeSizeMInY}
   Input High=CollectionParameter() {collection=&MPC_heatmap, parameterName=volumeSizeMaxY}
   Target Low=Constant()
   Target Height=Constant() {r=1}
  ) {materialFunction=&RemapValueRange}
  n6 = Floor(None=Multiply(A=@n5.Result, B=@n2) {constB=32})
  n7 = AppendVector(
   A=Add(A=Divide(A=@n3, B=@n2) {constB=32}, B=Divide(B=@n4) {constB=64}) {constB=0.0159}
   B=Add(A=Divide(A=@n6, B=@n2) {constB=32}, B=Divide(B=@n4) {constB=64})
  )
  FunctionOutput(None=MaterialFunctionCall(UVs=@n7, "Rotation Angle (0-1)"=Constant() {r=-0.25}) {materialFunction=&CustomRotator}.Rotated Values) {outputName=UV_B2_playerLoctaion, sortPriority=5}
  n8 = MaterialFunctionCall(
   Input=CollectionParameter() {collection=&MPC_heatmap, parameterName=PlayerX}
   Input Low=CollectionParameter() {collection=&MPC_heatmap, parameterName=volumeSizeMInX}
   Input High=CollectionParameter() {collection=&MPC_heatmap, parameterName=volumeSizeMaxX}
   Target Low=Constant()
   Target Height=Constant() {r=1}
  ) {materialFunction=&RemapValueRange}
  n9 = MaterialFunctionCall(
   Input=CollectionParameter() {collection=&MPC_heatmap, parameterName=PlayerY}
   Input Low=CollectionParameter() {collection=&MPC_heatmap, parameterName=volumeSizeMInY}
   Input High=CollectionParameter() {collection=&MPC_heatmap, parameterName=volumeSizeMaxY}
   Target Low=Constant()
   Target Height=Constant() {r=1}
  ) {materialFunction=&RemapValueRange}
  FunctionOutput(None=MaterialFunctionCall(UVs=AppendVector(A=@n8.Result, B=@n9.Result), "Rotation Angle (0-1)"=Constant() {r=-0.25}) {materialFunction=&CustomRotator}.Rotated Values) {outputName=Result_1}
  n10 = Multiply(A=Divide(B=CollectionParameter() {collection=&MPC_heatmap, parameterName=Gridsize})) {constB=3}
  FunctionOutput(None=MaterialFunctionCall(UVs=Add(A=@n7, B=@n10), "Rotation Angle (0-1)"=Constant() {r=-0.25}) {materialFunction=&CustomRotator}.Rotated Values) {outputName=UV_C3, sortPriority=9}
  n11 = MaterialFunctionCall(Float2=@n7) {materialFunction=&BreakOutFloat2Components}
  n12 = CollectionParameter() {collection=&MPC_heatmap, parameterName=Gridsize}
  n13 = Add(A=@n11.R, B=Multiply(A=Divide(B=@n12)) {constB=3})
  FunctionOutput(None=MaterialFunctionCall(UVs=AppendVector(A=@n13, B=@n11.G), "Rotation Angle (0-1)"=Constant() {r=-0.25}) {materialFunction=&CustomRotator}.Rotated Values) {outputName=UV_C2, sortPriority=8}
  n14 = MaterialFunctionCall(Float2=@n7) {materialFunction=&BreakOutFloat2Components}
  n15 = CollectionParameter() {collection=&MPC_heatmap, parameterName=Gridsize}
  n16 = Add(A=@n14.G, B=Multiply(A=Divide(B=@n15)) {constB=3})
  FunctionOutput(None=MaterialFunctionCall(UVs=AppendVector(A=@n14.R, B=@n16), "Rotation Angle (0-1)"=Constant() {r=-0.25}) {materialFunction=&CustomRotator}.Rotated Values) {outputName=UV_B3, sortPriority=6}
  n17 = MaterialFunctionCall(Float2=@n7) {materialFunction=&BreakOutFloat2Components}
  n18 = Divide(B=CollectionParameter() {collection=&MPC_heatmap, parameterName=Gridsize})
  n19 = Add(A=@n17.G, B=Multiply(A=Multiply(A=@n18) {constB=3}) {constB=-1})
  FunctionOutput(None=MaterialFunctionCall(UVs=AppendVector(A=@n17.R, B=@n19), "Rotation Angle (0-1)"=Constant() {r=-0.25}) {materialFunction=&CustomRotator}.Rotated Values) {outputName=UV_B1, sortPriority=4}
  n20 = MaterialFunctionCall(Float2=@n7) {materialFunction=&BreakOutFloat2Components}
  n21 = Divide(B=CollectionParameter() {collection=&MPC_heatmap, parameterName=Gridsize})
  n22 = Add(A=@n20.R, B=Multiply(A=Multiply(A=@n21) {constB=3}) {constB=-1})
  FunctionOutput(None=MaterialFunctionCall(UVs=AppendVector(A=@n22, B=@n20.G), "Rotation Angle (0-1)"=Constant() {r=-0.25}) {materialFunction=&CustomRotator}.Rotated Values) {outputName=UV_A2, sortPriority=1}
  n23 = MaterialFunctionCall(Float2=@n7) {materialFunction=&BreakOutFloat2Components}
  n24 = Multiply(A=Divide(B=CollectionParameter() {collection=&MPC_heatmap, parameterName=Gridsize})) {constB=3}
  n25 = Add(A=@n23.R, B=Multiply(A=@n24) {constB=-1})
  n26 = Add(A=@n23.G, B=@n24)
  FunctionOutput(None=MaterialFunctionCall(UVs=AppendVector(A=@n25, B=@n26), "Rotation Angle (0-1)"=Constant() {r=-0.25}) {materialFunction=&CustomRotator}.Rotated Values) {outputName=UV_A3, sortPriority=3}
  n27 = MaterialFunctionCall(Float2=@n7) {materialFunction=&BreakOutFloat2Components}
  n28 = Multiply(A=Divide(B=CollectionParameter() {collection=&MPC_heatmap, parameterName=Gridsize})) {constB=3}
  n29 = Add(A=@n27.R, B=@n28)
  n30 = Add(A=@n27.G, B=Multiply(A=@n28) {constB=-1})
  FunctionOutput(None=MaterialFunctionCall(UVs=AppendVector(A=@n29, B=@n30), "Rotation Angle (0-1)"=Constant() {r=-0.25}) {materialFunction=&CustomRotator}.Rotated Values) {outputName=UV_C1, sortPriority=7}
  n31 = FunctionInput(Preview=TextureCoordinate()) {bUsePreviewValueAsDefault=true, inputName=cell, inputType=FunctionInput_Vector2}
  n32 = MaterialFunctionCall(UVs=@n31, Texture Scale=Constant() {r=3.4}) {materialFunction=&ScaleUVsByCenter}
  n33 = TextureSample(UVs=@n32.UVs) {samplerSource=SSM_Clamp_WorldGroupSettings, texture=&/T_Dynamic_group}
  FunctionOutput(None=@n33.G) {outputName=MASK_DynamicGroupGrid}
  FunctionOutput(None=Multiply(A=@n33.B) {constB=2}) {outputName=MASK_DynamicCellTarget}
  FunctionOutput(None=TextureSample(UVs=@n32.UVs) {samplerSource=SSM_Clamp_WorldGroupSettings, texture=&/T_Dynamic_group_warning}.R) {outputName=MASK_DynamicGroupWarning}
  n34 = Multiply(
   A=TextureSample(UVs=@n31) {samplerSource=SSM_Clamp_WorldGroupSettings, texture=&/T_Footprint}.G
   B=VectorParameter() {parameterName=colorInter, defaultValue=<rgba 1 1 1 1>}.RGB
  )
  n35 = Multiply(A=Time(), B=ScalarParameter() {defaultValue=1, parameterName=PulseSpeed})
  n36 = Constant() {r=1}
  n37 = LinearInterpolate(Alpha=Power(Base=Fmod(A=@n35, B=@n36), Exp=ScalarParameter() {defaultValue=0.53, parameterName=FalloffPower})) {constA=1, constB=0.2}
  FunctionOutput(None=Saturate(None=Multiply(A=@n34, B=@n37))) {outputName=FX_Grid fx}
  FunctionOutput(None=TextureSample(UVs=@n32.UVs) {samplerSource=SSM_Clamp_WorldGroupSettings, texture=&/T_Dynamic_3DGrid}.R) {outputName=MASK_DynamicGroup3DGrid}
  FunctionOutput(None=TextureSample(UVs=@n32.UVs) {samplerSource=SSM_Clamp_WorldGroupSettings, texture=&/T_Static_group2}.R) {outputName=MASK_StaticGroupGrid2}
  FunctionOutput(None=MaterialFunctionCall(
    UVs=MaterialFunctionCall(X=@n1.Result, Y=@n5.Result) {materialFunction=&MakeFloat2}.Result
    "Rotation Angle (0-1)"=Constant() {r=-0.25}
   ) {materialFunction=&CustomRotator}.Rotated Values) {outputName=UV_UnsnappedPlayerLocation}
  n38 = CollectionParameter() {collection=&MPC_heatmap, parameterName=Gridsize}
  n39 = Multiply(
   A=MaterialFunctionCall(
    Input=CollectionParameter() {collection=&MPC_heatmap, parameterName=AimTargetX}
    Input Low=CollectionParameter() {collection=&MPC_heatmap, parameterName=volumeSizeMInX}
    Input High=CollectionParameter() {collection=&MPC_heatmap, parameterName=volumeSizeMaxX}
    Target Low=Constant()
    Target Height=Constant() {r=1}
   ) {materialFunction=&RemapValueRange}.Result
   B=@n38
  ) {constB=32}
  n40 = Multiply(A=@n38) {constB=2}
  n41 = Add(A=Divide(A=Floor(None=@n39), B=@n38) {constB=32}, B=Divide(B=@n40) {constB=64}) {constB=0.0159}
  n42 = Multiply(
   A=MaterialFunctionCall(
    Input=CollectionParameter() {collection=&MPC_heatmap, parameterName=AimTargetY}
    Input Low=CollectionParameter() {collection=&MPC_heatmap, parameterName=volumeSizeMInY}
    Input High=CollectionParameter() {collection=&MPC_heatmap, parameterName=volumeSizeMaxY}
    Target Low=Constant()
    Target Height=Constant() {r=1}
   ) {materialFunction=&RemapValueRange}.Result
   B=@n38
  ) {constB=32}
  n43 = Add(A=Divide(A=Floor(None=@n42), B=@n38) {constB=32}, B=Divide(B=@n40) {constB=64})
  FunctionOutput(None=MaterialFunctionCall(UVs=AppendVector(A=@n41, B=@n43), "Rotation Angle (0-1)"=Constant() {r=-0.25}) {materialFunction=&CustomRotator}.Rotated Values) {outputName=UV_AimLocation}
  n44 = CollectionParameter() {collection=&MPC_heatmap, parameterName=Gridsize}
  n45 = Multiply(
   A=MaterialFunctionCall(
    Input=CollectionParameter() {collection=&MPC_heatmap, parameterName=AimTargetX}
    Input Low=CollectionParameter() {collection=&MPC_heatmap, parameterName=volumeSizeMInX}
    Input High=CollectionParameter() {collection=&MPC_heatmap, parameterName=volumeSizeMaxX}
    Target Low=Constant()
    Target Height=Constant() {r=1}
   ) {materialFunction=&RemapValueRange}.Result
   B=@n44
  ) {constB=32}
  n46 = Multiply(A=@n44) {constB=2}
  n47 = Add(A=Divide(A=Floor(None=@n45), B=@n44) {constB=32}, B=Divide(B=@n46) {constB=64}) {constB=0.0159}
  n48 = Multiply(
   A=MaterialFunctionCall(
    Input=CollectionParameter() {collection=&MPC_heatmap, parameterName=AimTargetY}
    Input Low=CollectionParameter() {collection=&MPC_heatmap, parameterName=volumeSizeMInY}
    Input High=CollectionParameter() {collection=&MPC_heatmap, parameterName=volumeSizeMaxY}
    Target Low=Constant()
    Target Height=Constant() {r=1}
   ) {materialFunction=&RemapValueRange}.Result
   B=@n44
  ) {constB=32}
  n49 = Add(A=Divide(A=Floor(None=@n48), B=@n44) {constB=32}, B=Divide(B=@n46) {constB=64})
  FunctionOutput(None=MaterialFunctionCall(UVs=AppendVector(A=@n47, B=@n49), "Rotation Angle (0-1)"=Constant() {r=-0.25}) {materialFunction=&CustomRotator}.Rotated Values) {outputName=UV_AimTaregtPOS}
  n50 = MaterialFunctionCall(
   Input=CollectionParameter() {collection=&MPC_heatmap, parameterName=AimTargetX}
   Input Low=CollectionParameter() {collection=&MPC_heatmap, parameterName=volumeSizeMInX}
   Input High=CollectionParameter() {collection=&MPC_heatmap, parameterName=volumeSizeMaxX}
   Target Low=Constant()
   Target Height=Constant() {r=1}
  ) {materialFunction=&RemapValueRange}
  n51 = CollectionParameter() {collection=&MPC_heatmap, parameterName=Gridsize}
  n52 = Divide(A=Floor(None=Multiply(A=@n50.Result, B=@n51) {constB=32})) {constA=3, constB=3}
  n53 = Divide(B=@n51) {constA=3}
  n54 = Divide(B=@n51) {constA=1.5, constB=64}
  n55 = Add(A=Multiply(A=Floor(None=@n52), B=@n53), B=@n54) {constB=0.0159}
  n56 = MaterialFunctionCall(
   Input=CollectionParameter() {collection=&MPC_heatmap, parameterName=AimTargetY}
   Input Low=CollectionParameter() {collection=&MPC_heatmap, parameterName=volumeSizeMInY}
   Input High=CollectionParameter() {collection=&MPC_heatmap, parameterName=volumeSizeMaxY}
   Target Low=Constant()
   Target Height=Constant() {r=1}
  ) {materialFunction=&RemapValueRange}
  n57 = Divide(A=Floor(None=Multiply(A=@n51, B=@n56.Result) {constB=32})) {constA=3, constB=3}
  n58 = Add(A=Multiply(A=Floor(None=@n57), B=@n53), B=@n54)
  FunctionOutput(None=MaterialFunctionCall(UVs=AppendVector(A=@n55, B=@n58), "Rotation Angle (0-1)"=Constant() {r=-0.25}) {materialFunction=&CustomRotator}.Rotated Values) {outputName=UV_AimTargetGroupPos}
  n59 = MaterialFunctionCall(
   Input=CollectionParameter() {collection=&MPC_heatmap, parameterName=PlayerX}
   Input Low=CollectionParameter() {collection=&MPC_heatmap, parameterName=volumeSizeMInX}
   Input High=CollectionParameter() {collection=&MPC_heatmap, parameterName=volumeSizeMaxX}
   Target Low=Constant()
   Target Height=Constant() {r=1}
  ) {materialFunction=&RemapValueRange}
  n60 = CollectionParameter() {collection=&MPC_heatmap, parameterName=Gridsize}
  n61 = Divide(A=Floor(None=Multiply(A=@n59.Result, B=@n60) {constB=32})) {constA=3, constB=3}
  n62 = Divide(B=@n60) {constA=3}
  n63 = Divide(B=@n60) {constA=1.5, constB=64}
  n64 = Add(A=Multiply(A=Floor(None=@n61), B=@n62), B=@n63) {constB=0.0159}
  n65 = MaterialFunctionCall(
   Input=CollectionParameter() {collection=&MPC_heatmap, parameterName=PlayerY}
   Input Low=CollectionParameter() {collection=&MPC_heatmap, parameterName=volumeSizeMInY}
   Input High=CollectionParameter() {collection=&MPC_heatmap, parameterName=volumeSizeMaxY}
   Target Low=Constant()
   Target Height=Constant() {r=1}
  ) {materialFunction=&RemapValueRange}
  n66 = Divide(A=Floor(None=Multiply(A=@n60, B=@n65.Result) {constB=32})) {constA=3, constB=3}
  n67 = Add(A=Multiply(A=Floor(None=@n66), B=@n62), B=@n63)
  FunctionOutput(None=MaterialFunctionCall(UVs=AppendVector(A=@n64, B=@n67), "Rotation Angle (0-1)"=Constant() {r=-0.25}) {materialFunction=&CustomRotator}.Rotated Values) {outputName=UV_staticGroupPos}
  n68 = TextureSample(UVs=MaterialFunctionCall(UVs=@n31, Texture Scale=Constant() {r=3}) {materialFunction=&ScaleUVsByCenter}.UVs) {samplerSource=SSM_Clamp_WorldGroupSettings, texture=&/T_DynamicFootprint}
  FunctionOutput(None=@n68.R) {outputName=MASK_DynamicGroup}
  FunctionOutput(None=@n68.B) {outputName=MASK_DynamicGroupOutline}
  n69 = Multiply(A=Divide(B=CollectionParameter() {collection=&MPC_heatmap, parameterName=Gridsize})) {constB=3}
  FunctionOutput(None=MaterialFunctionCall(UVs=Subtract(A=@n7, B=@n69), "Rotation Angle (0-1)"=Constant() {r=-0.25}) {materialFunction=&CustomRotator}.Rotated Values) {outputName=UV_A1, sortPriority=10}
  FunctionOutput(None=TextureSample(UVs=@n32.UVs) {samplerSource=SSM_Clamp_WorldGroupSettings, texture=&/T_Dynamic_groupMask}.R) {outputName=MASK_DynamicGroupCoordinate}
  n70 = Divide(
   A=Subtract(
    A=CollectionParameter() {collection=&MPC_heatmap, parameterName=volumeSizeMaxX}
    B=CollectionParameter() {collection=&MPC_heatmap, parameterName=volumeSizeMInX}
   )
   B=CollectionParameter() {collection=&MPC_heatmap, parameterName=Gridsize}
  )
  n71 = MaterialFunctionCall(
   Input=Subtract(
    A=CollectionParameter() {collection=&MPC_heatmap, parameterName=PlayerX}
    B=Multiply(A=@n70) {constB=3}
   )
   Input Low=CollectionParameter() {collection=&MPC_heatmap, parameterName=volumeSizeMInX}
   Input High=CollectionParameter() {collection=&MPC_heatmap, parameterName=volumeSizeMaxX}
   Target Low=Constant()
   Target Height=Constant() {r=1}
  ) {materialFunction=&RemapValueRange}
  n72 = CollectionParameter() {collection=&MPC_heatmap, parameterName=Gridsize}
  n73 = Divide(A=Floor(None=Multiply(A=@n71.Result, B=@n72) {constB=32}), B=@n72) {constB=32}
  n74 = Multiply(A=@n72) {constB=2}
  n75 = Divide(B=@n74) {constB=64}
  n76 = Divide(
   A=Subtract(
    A=CollectionParameter() {collection=&MPC_heatmap, parameterName=volumeSizeMaxY}
    B=CollectionParameter() {collection=&MPC_heatmap, parameterName=volumeSizeMInY}
   )
   B=CollectionParameter() {collection=&MPC_heatmap, parameterName=Gridsize}
  )
  n77 = MaterialFunctionCall(
   Input=Subtract(
    A=CollectionParameter() {collection=&MPC_heatmap, parameterName=PlayerY}
    B=Multiply(A=@n76) {constB=3}
   )
   Input Low=CollectionParameter() {collection=&MPC_heatmap, parameterName=volumeSizeMInY}
   Input High=CollectionParameter() {collection=&MPC_heatmap, parameterName=volumeSizeMaxY}
   Target Low=Constant()
   Target Height=Constant() {r=1}
  ) {materialFunction=&RemapValueRange}
  n78 = Divide(A=Floor(None=Multiply(A=@n77.Result, B=@n72) {constB=32}), B=@n72) {constB=32}
  n79 = Divide(B=@n74) {constB=64}
  MaterialFunctionCall(UVs=AppendVector(A=Add(A=@n73, B=@n75) {constB=0.0159}, B=Add(A=@n78, B=@n79)), "Rotation Angle (0-1)"=Constant() {r=-0.25}) {materialFunction=&CustomRotator}
  n80 = CollectionParameter() {collection=&MPC_heatmap, parameterName=volumeSizeMaxY}
  n81 = CollectionParameter() {collection=&MPC_heatmap, parameterName=volumeSizeMInY}
  Multiply(A=Divide(
    A=Subtract(A=@n80, B=@n81)
    B=CollectionParameter() {collection=&MPC_heatmap, parameterName=Gridsize}
   )) {constB=3}
 ]}
