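# Material function for the Playground Thermometer LTM HUD that builds a combined dynamic-group mask over a 3x3 minimap grid (A1 to C3, B2 named player location).
# For each cell, it scales input UVs about center by MPC_heatmap's minimap scale and Gridsize, then offsets them by that cell's MF_minimapUIUV coordinate.
# Each resulting MF_minimapUIUV mask is weighted by its matching collection parameter, and all nine are summed into one output.
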MaterialFunction /Game/Athena/HUD/LTM/Playground/Thermometer/MF_DynamicGroupCoordinate
graph {nodes [
  n1 = FunctionInput(Preview=TextureCoordinate()) {inputType=FunctionInput_Vector2}
  n2 = MaterialFunctionCall(
   UVs=@n1
   Texture Scale=CollectionParameter() {collection=&/MPC_heatmap, parameterName=MInnimapUvScale}
  ) {materialFunction=&ScaleUVsByCenter}
  n3 = Divide(B=CollectionParameter() {collection=&MPC_heatmap, parameterName=Gridsize}) {constA=3, constB=32}
  n4 = Multiply(A=CollectionParameter() {collection=&MPC_heatmap, parameterName=Gridsize}) {constB=0.5}
  n5 = CollectionParameter() {collection=&MPC_heatmap, parameterName=Gridsize}
  n6 = Subtract(
   A=Add(
    A=MaterialFunctionCall(UVs=@n2.UVs, Texture Scale=@n3) {materialFunction=&ScaleUVsByCenter}.UVs
    B=Divide(A=@n4) {constA=16, constB=3}
   ) {constB=3}
   B=Multiply(
    A=MaterialFunctionCall!(cell) {materialFunction=&/MF_minimapUIUV}.UV_A2
    B=Divide(A=@n5) {constA=32, constB=3}
   ) {constB=3}
  )
  n7 = MaterialFunctionCall(
   UVs=@n1
   Texture Scale=CollectionParameter() {collection=&MPC_heatmap, parameterName=MInnimapUvScale}
  ) {materialFunction=&ScaleUVsByCenter}
  n8 = Divide(B=CollectionParameter() {collection=&MPC_heatmap, parameterName=Gridsize}) {constA=3, constB=32}
  n9 = Multiply(A=CollectionParameter() {collection=&MPC_heatmap, parameterName=Gridsize}) {constB=0.5}
  n10 = CollectionParameter() {collection=&MPC_heatmap, parameterName=Gridsize}
  n11 = Subtract(
   A=Add(
    A=MaterialFunctionCall(UVs=@n7.UVs, Texture Scale=@n8) {materialFunction=&ScaleUVsByCenter}.UVs
    B=Divide(A=@n9) {constA=16, constB=3}
   ) {constB=3}
   B=Multiply(
    A=MaterialFunctionCall() {materialFunction=&MF_minimapUIUV}.UV_A3
    B=Divide(A=@n10) {constA=32, constB=3}
   ) {constB=3}
  )
  n12 = Add(
   A=LinearInterpolate(
    B=MaterialFunctionCall(cell=@n6) {materialFunction=&MF_minimapUIUV}.MASK_DynamicGroup
    Alpha=CollectionParameter() {collection=&MPC_heatmap, parameterName=A2}
   )
   B=LinearInterpolate(
    B=MaterialFunctionCall(cell=@n11) {materialFunction=&MF_minimapUIUV}.MASK_DynamicGroup
    Alpha=CollectionParameter() {collection=&MPC_heatmap, parameterName=A3}
   )
  )
  n13 = MaterialFunctionCall(
   UVs=@n1
   Texture Scale=CollectionParameter() {collection=&MPC_heatmap, parameterName=MInnimapUvScale}
  ) {materialFunction=&ScaleUVsByCenter}
  n14 = Divide(B=CollectionParameter() {collection=&MPC_heatmap, parameterName=Gridsize}) {constA=3, constB=32}
  n15 = Multiply(A=CollectionParameter() {collection=&MPC_heatmap, parameterName=Gridsize}) {constB=0.5}
  n16 = CollectionParameter() {collection=&MPC_heatmap, parameterName=Gridsize}
  n17 = Subtract(
   A=Add(
    A=MaterialFunctionCall(UVs=@n13.UVs, Texture Scale=@n14) {materialFunction=&ScaleUVsByCenter}.UVs
    B=Divide(A=@n15) {constA=16, constB=3}
   ) {constB=3}
   B=Multiply(
    A=MaterialFunctionCall() {materialFunction=&MF_minimapUIUV}.UV_B1
    B=Divide(A=@n16) {constA=32, constB=3}
   ) {constB=3}
  )
  n18 = MaterialFunctionCall(
   UVs=@n1
   Texture Scale=CollectionParameter() {collection=&MPC_heatmap, parameterName=MInnimapUvScale}
  ) {materialFunction=&ScaleUVsByCenter}
  n19 = Divide(B=CollectionParameter() {collection=&MPC_heatmap, parameterName=Gridsize}) {constA=3, constB=32}
  n20 = Multiply(A=CollectionParameter() {collection=&MPC_heatmap, parameterName=Gridsize}) {constB=0.5}
  n21 = CollectionParameter() {collection=&MPC_heatmap, parameterName=Gridsize}
  n22 = Subtract(
   A=Add(
    A=MaterialFunctionCall(UVs=@n18.UVs, Texture Scale=@n19) {materialFunction=&ScaleUVsByCenter}.UVs
    B=Divide(A=@n20) {constA=16, constB=3}
   ) {constB=3}
   B=Multiply(
    A=MaterialFunctionCall() {materialFunction=&MF_minimapUIUV}.UV_B2_playerLoctaion
    B=Divide(A=@n21) {constA=32, constB=3}
   ) {constB=3}
  )
  n23 = Add(
   A=LinearInterpolate(
    B=MaterialFunctionCall(cell=@n17) {materialFunction=&MF_minimapUIUV}.MASK_DynamicGroup
    Alpha=CollectionParameter() {collection=&MPC_heatmap, parameterName=B1}
   )
   B=LinearInterpolate(
    B=MaterialFunctionCall(cell=@n22) {materialFunction=&MF_minimapUIUV}.MASK_DynamicGroup
    Alpha=CollectionParameter() {collection=&MPC_heatmap, parameterName=B2}
   )
  )
  n24 = MaterialFunctionCall(
   UVs=@n1
   Texture Scale=CollectionParameter() {collection=&MPC_heatmap, parameterName=MInnimapUvScale}
  ) {materialFunction=&ScaleUVsByCenter}
  n25 = Divide(B=CollectionParameter() {collection=&MPC_heatmap, parameterName=Gridsize}) {constA=3, constB=32}
  n26 = Multiply(A=CollectionParameter() {collection=&MPC_heatmap, parameterName=Gridsize}) {constB=0.5}
  n27 = CollectionParameter() {collection=&MPC_heatmap, parameterName=Gridsize}
  n28 = Subtract(
   A=Add(
    A=MaterialFunctionCall(UVs=@n24.UVs, Texture Scale=@n25) {materialFunction=&ScaleUVsByCenter}.UVs
    B=Divide(A=@n26) {constA=16, constB=3}
   ) {constB=3}
   B=Multiply(
    A=MaterialFunctionCall() {materialFunction=&MF_minimapUIUV}.UV_B3
    B=Divide(A=@n27) {constA=32, constB=3}
   ) {constB=3}
  )
  n29 = MaterialFunctionCall(
   UVs=@n1
   Texture Scale=CollectionParameter() {collection=&MPC_heatmap, parameterName=MInnimapUvScale}
  ) {materialFunction=&ScaleUVsByCenter}
  n30 = Divide(B=CollectionParameter() {collection=&MPC_heatmap, parameterName=Gridsize}) {constA=3, constB=32}
  n31 = Multiply(A=CollectionParameter() {collection=&MPC_heatmap, parameterName=Gridsize}) {constB=0.5}
  n32 = CollectionParameter() {collection=&MPC_heatmap, parameterName=Gridsize}
  n33 = Subtract(
   A=Add(
    A=MaterialFunctionCall(UVs=@n29.UVs, Texture Scale=@n30) {materialFunction=&ScaleUVsByCenter}.UVs
    B=Divide(A=@n31) {constA=16, constB=3}
   ) {constB=3}
   B=Multiply(
    A=MaterialFunctionCall() {materialFunction=&MF_minimapUIUV}.UV_C1
    B=Divide(A=@n32) {constA=32, constB=3}
   ) {constB=3}
  )
  n34 = Add(
   A=LinearInterpolate(
    B=MaterialFunctionCall(cell=@n28) {materialFunction=&MF_minimapUIUV}.MASK_DynamicGroup
    Alpha=CollectionParameter() {collection=&MPC_heatmap, parameterName=B3}
   )
   B=LinearInterpolate(
    B=MaterialFunctionCall(cell=@n33) {materialFunction=&MF_minimapUIUV}.MASK_DynamicGroup
    Alpha=CollectionParameter() {collection=&MPC_heatmap, parameterName=C1}
   )
  )
  n35 = MaterialFunctionCall(
   UVs=MaterialFunctionCall(
    UVs=@n1
    Texture Scale=CollectionParameter() {collection=&MPC_heatmap, parameterName=MInnimapUvScale}
   ) {materialFunction=&ScaleUVsByCenter}.UVs
   Texture Scale=Divide(B=CollectionParameter() {collection=&MPC_heatmap, parameterName=Gridsize}) {constA=3, constB=32}
  ) {materialFunction=&ScaleUVsByCenter}
  n36 = Divide(A=Multiply(A=CollectionParameter() {collection=&MPC_heatmap, parameterName=Gridsize}) {constB=0.5}) {constA=16, constB=3}
  n37 = MaterialFunctionCall() {materialFunction=&MF_minimapUIUV}
  n38 = Divide(A=CollectionParameter() {collection=&MPC_heatmap, parameterName=Gridsize}) {constA=32, constB=3}
  n39 = MaterialFunctionCall(cell=Subtract(A=Add(A=@n35.UVs, B=@n36) {constB=3}, B=Multiply(A=@n37.UV_C2, B=@n38) {constB=3})) {materialFunction=&MF_minimapUIUV}
  n40 = CollectionParameter() {collection=&MPC_heatmap, parameterName=C2}
  n41 = MaterialFunctionCall(
   UVs=MaterialFunctionCall(
    UVs=@n1
    Texture Scale=CollectionParameter() {collection=&MPC_heatmap, parameterName=MInnimapUvScale}
   ) {materialFunction=&ScaleUVsByCenter}.UVs
   Texture Scale=Divide(B=CollectionParameter() {collection=&MPC_heatmap, parameterName=Gridsize}) {constA=3, constB=32}
  ) {materialFunction=&ScaleUVsByCenter}
  n42 = Divide(A=Multiply(A=CollectionParameter() {collection=&MPC_heatmap, parameterName=Gridsize}) {constB=0.5}) {constA=16, constB=3}
  n43 = MaterialFunctionCall() {materialFunction=&MF_minimapUIUV}
  n44 = Divide(A=CollectionParameter() {collection=&MPC_heatmap, parameterName=Gridsize}) {constA=32, constB=3}
  n45 = MaterialFunctionCall(cell=Subtract(A=Add(A=@n41.UVs, B=@n42) {constB=3}, B=Multiply(A=@n43.UV_C3, B=@n44) {constB=3})) {materialFunction=&MF_minimapUIUV}
  n46 = CollectionParameter() {collection=&MPC_heatmap, parameterName=C3}
  n47 = MaterialFunctionCall(
   UVs=@n1
   Texture Scale=CollectionParameter() {collection=&MPC_heatmap, parameterName=MInnimapUvScale}
  ) {materialFunction=&ScaleUVsByCenter}
  n48 = Divide(B=CollectionParameter() {collection=&MPC_heatmap, parameterName=Gridsize}) {constA=3, constB=32}
  n49 = Multiply(A=CollectionParameter() {collection=&MPC_heatmap, parameterName=Gridsize}) {constB=0.5}
  n50 = CollectionParameter() {collection=&MPC_heatmap, parameterName=Gridsize}
  n51 = Subtract(
   A=Add(
    A=MaterialFunctionCall(UVs=@n47.UVs, Texture Scale=@n48) {materialFunction=&ScaleUVsByCenter}.UVs
    B=Divide(A=@n49) {constA=16, constB=3}
   ) {constB=3}
   B=Multiply(
    A=MaterialFunctionCall() {materialFunction=&MF_minimapUIUV}.UV_A1
    B=Divide(A=@n50) {constA=32, constB=3}
   ) {constB=3}
  )
  n52 = Add(
   A=Add(A=LinearInterpolate(B=@n39.MASK_DynamicGroup, Alpha=@n40), B=LinearInterpolate(B=@n45.MASK_DynamicGroup, Alpha=@n46))
   B=LinearInterpolate(
    B=MaterialFunctionCall(cell=@n51) {materialFunction=&MF_minimapUIUV}.MASK_DynamicGroup
    Alpha=CollectionParameter() {collection=&MPC_heatmap, parameterName=A1}
   )
  )
  FunctionOutput(None=Add(A=Add(A=@n12, B=@n23), B=Add(A=@n34, B=@n52)))
  TextureCoordinate()
  TextureCoordinate()
  TextureCoordinate()
  TextureCoordinate()
  TextureCoordinate()
  TextureCoordinate()
  TextureCoordinate()
  TextureCoordinate()
 ]}
