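# Material function that converts world position into UVs for a heatmap volume, filed under the Playground LTM thermometer HUD.
# World X and Y are each remapped to 0-1 using the volume min and max bounds from the MPC_heatmap parameter collection.
# The two results are appended and passed through CustomRotator for a negative quarter turn before output.
# A OneMinus on the remapped Y is left unconnected to the output.
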MaterialFunction /Game/Athena/HUD/LTM/Playground/Thermometer/MF_WorldtoUV
graph {nodes [
  n1 = MaterialFunctionCall(Float2=ComponentMask(None=WorldPosition().XYZ) {g=true, r=true}) {materialFunction=&BreakOutFloat2Components}
  n2 = MaterialFunctionCall!(
   Input=@n1.R
   Input Low=CollectionParameter() {collection=&/MPC_heatmap, parameterName=volumeSizeMInX}
   Input High=CollectionParameter() {collection=&MPC_heatmap, parameterName=volumeSizeMaxX}
   Target Low=Constant()
   Target Height=Constant() {r=1}
  ) {materialFunction=&RemapValueRange}
  n3 = MaterialFunctionCall(
   Input=@n1.G
   Input Low=CollectionParameter() {collection=&MPC_heatmap, parameterName=volumeSizeMInY}
   Input High=CollectionParameter() {collection=&MPC_heatmap, parameterName=volumeSizeMaxY}
   Target Low=Constant()
   Target Height=Constant() {r=1}
  ) {materialFunction=&RemapValueRange}
  FunctionOutput(None=MaterialFunctionCall(UVs=AppendVector(A=@n2.Result, B=@n3.Result), "Rotation Angle (0-1)"=Constant() {r=-0.25}) {materialFunction=&CustomRotator}.Rotated Values)
  OneMinus(None=@n3.Result)
 ]}
