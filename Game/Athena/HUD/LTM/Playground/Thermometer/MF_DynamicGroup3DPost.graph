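# Material function that draws a world-projected dynamic group grid overlay for the Playground LTM thermometer HUD, driven by MPC_heatmap.
# World-to-UV coordinates sample grid and lock textures inside sphere masks around the player, plus warning graphics, some gated by RT_Heatmapcode exceeding CellThreshold.
# The output adds group numbers brightened near the aimed group and an aimed-cell highlight cut out around the player.
# Several grid, coordinate and fade branches are left unconnected.
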MaterialFunction /Game/Athena/HUD/LTM/Playground/Thermometer/MF_DynamicGroup3DPost
graph {nodes [
  n1 = Constant3Vector() {constant=<rgba 0 2 0 0>}
  n2 = MaterialFunctionCall(
   UVs=MaterialFunctionCall(
    UVs=MaterialFunctionCall() {materialFunction=&/MF_WorldtoUV}.Result
    Texture Scale=CollectionParameter() {collection=&/MPC_heatmap, parameterName=MInnimapUvScale}
   ) {materialFunction=&ScaleUVsByCenter}.UVs
   Texture Scale=Divide(B=CollectionParameter() {collection=&MPC_heatmap, parameterName=Gridsize}) {constA=3, constB=32}
  ) {materialFunction=&ScaleUVsByCenter}
  n3 = Divide(A=Multiply(A=CollectionParameter() {collection=&MPC_heatmap, parameterName=Gridsize}) {constB=0.5}) {constA=16, constB=3}
  n4 = MaterialFunctionCall!(cell) {materialFunction=&/MF_minimapUIUV}
  n5 = Divide(A=CollectionParameter() {collection=&MPC_heatmap, parameterName=Gridsize}) {constA=32, constB=3}
  n6 = MaterialFunctionCall(cell=Subtract(A=Add(A=@n2.UVs, B=@n3) {constB=3}, B=Multiply(A=@n4.UV_B2_playerLoctaion, B=@n5) {constB=3})) {materialFunction=&MF_minimapUIUV}
  n7 = MaterialFunctionCall(
   UVs=MaterialFunctionCall() {materialFunction=&MF_WorldtoUV}.Result
   Texture Scale=CollectionParameter() {collection=&MPC_heatmap, parameterName=MInnimapUvScale}
  ) {materialFunction=&ScaleUVsByCenter}
  n8 = MaterialFunctionCall() {materialFunction=&MF_minimapUIUV}
  n9 = Multiply(A=Divide(B=CollectionParameter() {collection=&MPC_heatmap, parameterName=Gridsize})) {constB=2.12}
  n10 = Constant() {r=1}
  n11 = MaterialFunctionCall() {materialFunction=&MF_WorldtoUV}
  n12 = MaterialFunctionCall(UVs=@n11.Result, "Rotation Angle (0-1)"=Constant() {r=0.25}) {materialFunction=&CustomRotator}
  n13 = CollectionParameter() {collection=&MPC_heatmap, parameterName=Gridsize}
  n14 = Multiply(
   A=TextureSample(UVs=Multiply(A=@n12.Rotated Values, B=@n13)) {texture=&/T_CreativeMoveTool_Grid}.R
   B=Constant3Vector() {constant=<rgba 0.9479 0.9479 0.9479 0>}
  )
  n15 = CollectionParameter() {collection=&MPC_heatmap, parameterName=Gridsize}
  n16 = SphereMask(
   A=MaterialFunctionCall(
    UVs=@n11.Result
    Texture Scale=CollectionParameter() {collection=&MPC_heatmap, parameterName=MInnimapUvScale}
   ) {materialFunction=&ScaleUVsByCenter}.UVs
   B=MaterialFunctionCall() {materialFunction=&MF_minimapUIUV}.UV_B2_playerLoctaion
   Radius=Multiply(A=Divide(B=@n15)) {constB=0.65}
   Hardness=Constant() {r=1}
  )
  n17 = MaterialFunctionCall() {materialFunction=&MF_WorldtoUV}
  n18 = MaterialFunctionCall(UVs=@n17.Result, "Rotation Angle (0-1)"=Constant() {r=0.25}) {materialFunction=&CustomRotator}
  n19 = CollectionParameter() {collection=&MPC_heatmap, parameterName=Gridsize}
  n20 = Multiply(
   A=TextureSample(UVs=Multiply(A=@n18.Rotated Values, B=@n19)) {texture=&/T_Lock_group}.G
   B=Constant3Vector() {constant=<rgba 1.5 1.5 1.5 0>}
  )
  n21 = CollectionParameter() {collection=&MPC_heatmap, parameterName=Gridsize}
  n22 = SphereMask(
   A=MaterialFunctionCall(
    UVs=@n17.Result
    Texture Scale=CollectionParameter() {collection=&MPC_heatmap, parameterName=MInnimapUvScale}
   ) {materialFunction=&ScaleUVsByCenter}.UVs
   B=MaterialFunctionCall() {materialFunction=&MF_minimapUIUV}.UV_B2_playerLoctaion
   Radius=Multiply(A=Divide(B=@n21)) {constB=0.712}
   Hardness=Constant() {r=1}
  )
  n23 = Multiply(A=@n6.MASK_DynamicGroupWarning, B=Constant3Vector() {constant=<rgba 2.6758 3.7192 4 0>}) {constB=20}
  n24 = MaterialFunctionCall() {materialFunction=&MF_WorldtoUV}
  n25 = MaterialFunctionCall!(In=@n24.Result) {materialFunction=&/MF_DynamicGroupCoordinate}
  n26 = MaterialFunctionCall(UVs=@n24.Result, "Rotation Angle (0-1)"=Constant() {r=0.25}) {materialFunction=&CustomRotator}
  n27 = CollectionParameter() {collection=&MPC_heatmap, parameterName=Gridsize}
  n28 = Multiply(
   A=TextureSample(UVs=Multiply(A=@n26.Rotated Values, B=@n27)) {texture=&/T_GroupWarning}.G
   B=Constant3Vector() {constant=<rgba 5 0 0 0>}
  ) {constB=20}
  n29 = TextureSample(UVs=MaterialFunctionCall() {materialFunction=&MF_WorldtoUV}.Result) {
   samplerSource=SSM_Clamp_WorldGroupSettings
   samplerType=SAMPLERTYPE_LinearColor
   texture=&/RT_Heatmapcode
  }
  n30 = Constant() {r=1}
  n31 = Multiply(
   A=@n25.Result
   B=If(
    A=@n29.R
    B=CollectionParameter() {collection=&MPC_heatmap, parameterName=CellThreshold}
    "A > B"=@n30
    "A == B"=@n30
    "A < B"=Constant()
   )
  )
  n32 = Add(A=Add(A=Multiply(A=@n14, B=@n16), B=Multiply(A=@n20, B=@n22)), B=Add(A=Multiply(A=@n23, B=@n25.Result), B=Multiply(A=@n28, B=@n31)))
  Add(
   A=Multiply(A=Multiply(A=@n6.MASK_DynamicGroupGrid, B=@n1), B=SphereMask(A=@n7.UVs, B=@n8.UV_B2_playerLoctaion, Radius=@n9, Hardness=@n10))
   B=@n32
  )
  Add(A=Multiply(A=@n6.MASK_DynamicGroupCoordinate, B=Constant3Vector() {constant=<rgba 1 1 1 0>}), B=@n32)
  n33 = MaterialFunctionCall(
   UVs=MaterialFunctionCall() {materialFunction=&MF_WorldtoUV}.Result
   Texture Scale=CollectionParameter() {collection=&MPC_heatmap, parameterName=MInnimapUvScale}
  ) {materialFunction=&ScaleUVsByCenter}
  n34 = MaterialFunctionCall() {materialFunction=&MF_minimapUIUV}
  n35 = Multiply(A=Divide(B=CollectionParameter() {collection=&MPC_heatmap, parameterName=Gridsize}) {constB=24}) {constB=0.65}
  n36 = Constant() {r=1}
  n37 = MaterialFunctionCall(
   UVs=MaterialFunctionCall() {materialFunction=&MF_WorldtoUV}.Result
   Texture Scale=CollectionParameter() {collection=&MPC_heatmap, parameterName=MInnimapUvScale}
  ) {materialFunction=&ScaleUVsByCenter}
  n38 = MaterialFunctionCall() {materialFunction=&MF_minimapUIUV}
  n39 = Multiply(A=Divide(B=CollectionParameter() {collection=&MPC_heatmap, parameterName=Gridsize})) {constB=0.5}
  n40 = Constant() {r=1}
  n41 = Saturate(None=Subtract(
    A=SphereMask(A=@n33.UVs, B=@n34.UV_AimTaregtPOS, Radius=@n35, Hardness=@n36)
    B=SphereMask(A=@n37.UVs, B=@n38.UV_B2_playerLoctaion, Radius=@n39, Hardness=@n40)
   ))
  n42 = MaterialFunctionCall() {materialFunction=&MF_WorldtoUV}
  n43 = Constant() {r=0.25}
  n44 = Divide(A=CollectionParameter() {collection=&MPC_heatmap, parameterName=Gridsize}) {constB=3}
  n45 = TextureSample(UVs=Multiply(
    A=MaterialFunctionCall(UVs=@n42.Result, "Rotation Angle (0-1)"=@n43) {materialFunction=&CustomRotator}.Rotated Values
    B=Divide(A=@n44) {constB=10}
   )) {
   samplerSource=SSM_Clamp_WorldGroupSettings
   samplerType=SAMPLERTYPE_LinearColor
   texture=&/GroupNUM
  }
  n46 = MaterialFunctionCall(
   UVs=@n42.Result
   Texture Scale=CollectionParameter() {collection=&MPC_heatmap, parameterName=MInnimapUvScale}
  ) {materialFunction=&ScaleUVsByCenter}
  n47 = MaterialFunctionCall() {materialFunction=&MF_minimapUIUV}
  n48 = Multiply(A=Divide(B=CollectionParameter() {collection=&MPC_heatmap, parameterName=Gridsize}) {constB=24}) {constB=2.12}
  n49 = Constant() {r=0.5}
  n50 = Multiply(
   A=@n45.R
   B=LinearInterpolate(
    A=Constant3Vector() {constant=<rgba 0.2698 0.3001 0.35 0>}
    B=Constant3Vector() {constant=<rgba 1.688 1.8059 2 0>}
    Alpha=SphereMask(A=@n46.UVs, B=@n47.UV_AimTargetGroupPos, Radius=@n48, Hardness=@n49)
   )
  )
  FunctionOutput(None=Add(A=Multiply(A=@n41, B=@n6.MASK_DynamicCellTarget), B=Add(A=@n50, B=@n32)))
  n51 = MaterialFunctionCall(
   UVs=@n42.Result
   Texture Scale=CollectionParameter() {collection=&MPC_heatmap, parameterName=MInnimapUvScale}
  ) {materialFunction=&ScaleUVsByCenter}
  n52 = Constant() {r=0.5}
  n53 = Constant() {r=0.5}
  n54 = Constant() {r=0.8}
  LinearInterpolate(A=Multiply(A=SphereMask(A=@n51.UVs, B=@n52, Radius=@n53, Hardness=@n54)) {constB=0.35}) {constB=2}
  Multiply(A=Divide(B=CollectionParameter() {collection=&MPC_heatmap, parameterName=Gridsize})) {constB=2.12}
  TextureCoordinate()
 ]}
