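# Material function that draws a group-radius ring around the player location for the Playground Thermometer heatmap, apparently in a post-process pass.
# World-derived, center-scaled UVs feed sphere masks sized by MPC_heatmap GroupRadius and Gridsize.
# Subtracted masks form a faint thin outline plus a soft inner glow, tinted blue, green, orange or red by GroupNewValue.
# GroupRadiusSwitch blends the result in; the ThrowItems center blend has identical inputs, so it has no effect.
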MaterialFunction /Game/Athena/HUD/LTM/Playground/Thermometer/MF_StaticGroup3DPost
graph {nodes [
  n1 = MaterialFunctionCall() {materialFunction=&/MF_WorldtoUV}
  n2 = MaterialFunctionCall!(cell) {materialFunction=&/MF_minimapUIUV}
  n3 = LinearInterpolate(
   A=@n2.UV_B2_playerLoctaion
   B=@n2.UV_B2_playerLoctaion
   Alpha=CollectionParameter() {collection=&/MPC_heatmap, parameterName=ThrowItems}
  )
  n4 = Divide(B=CollectionParameter() {collection=&MPC_heatmap, parameterName=Gridsize})
  n5 = CollectionParameter() {collection=&MPC_heatmap, parameterName=GroupRadius}
  n6 = Add(A=Multiply(A=Multiply(A=@n5) {constB=2}) {constB=0.5}) {constB=-0.5}
  n7 = Multiply(A=@n4, B=@n6) {constB=2.225}
  n8 = SphereMask(
   A=MaterialFunctionCall(
    UVs=@n1.Result
    Texture Scale=CollectionParameter() {collection=&MPC_heatmap, parameterName=MInnimapUvScale}
   ) {materialFunction=&ScaleUVsByCenter}.UVs
   B=@n3
   Radius=@n7
   Hardness=Constant() {r=0.995}
  )
  n9 = MaterialFunctionCall(
   UVs=@n1.Result
   Texture Scale=CollectionParameter() {collection=&MPC_heatmap, parameterName=MInnimapUvScale}
  ) {materialFunction=&ScaleUVsByCenter}
  n10 = Subtract(A=@n7) {constB=0.003}
  n11 = Constant() {r=0.995}
  n12 = Multiply(A=Subtract(A=@n8, B=SphereMask(A=@n9.UVs, B=@n3, Radius=@n10, Hardness=@n11))) {constB=0.07}
  n13 = Divide(B=CollectionParameter() {collection=&MPC_heatmap, parameterName=Gridsize})
  n14 = Multiply(A=@n13, B=Subtract(A=@n6) {constB=0.01}) {constB=2.225}
  n15 = Constant() {r=0.5}
  n16 = Power(Base=Subtract(A=@n8, B=SphereMask(A=@n9.UVs, B=@n3, Radius=@n14, Hardness=@n15))) {constExponent=6}
  n17 = Multiply(
   A=Multiply(A=Saturate(None=@n16)) {constB=0.85}
   B=MaterialFunctionCall(
    1=Constant3Vector() {constant=<rgba 0 0 2.5 0>}
    2=Constant3Vector() {constant=<rgba 0 1 0 1>}
    3=Constant3Vector() {constant=<rgba 1 0.6776 0 0>}
    4=Constant3Vector() {constant=<rgba 1 0.03 0.03 0>}
    A=CollectionParameter() {collection=&MPC_heatmap, parameterName=GroupNewValue}
   ) {materialFunction=&Lerp_Multiple_Float3}.Lerp 4 Inputs
  )
  FunctionOutput(None=LinearInterpolate(
    B=Add(A=@n12, B=@n17)
    Alpha=CollectionParameter() {collection=&MPC_heatmap, parameterName=GroupRadiusSwitch}
   ))
  Multiply(A=@n4) {constB=0.5}
  Multiply(A=@n13) {constB=0.5}
  TextureCoordinate()
  TextureCoordinate()
 ]}
