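# Post-process material for the Playground thermometer that overlays a memory heatmap on the world in a camera-centred sphere, driven by MPC_heatmap.
# World-to-UV sampling of RT_Heatmapcode gives a blue-to-red ramp with pulsing warning and lock icons.
# The sphere's radius drives desaturation, grid-textured wavefront rings and fade-in, while switches toggle static or dynamic group grids and ThrowItems masking.
# A depth comparison leaves visible custom-depth objects unaffected.
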Material /Game/Athena/HUD/LTM/Playground/Thermometer/M_PostProcessIngame
graph {
 properties {
  blendableLocation=BL_SceneColorAfterDOF
  materialDomain=MD_PostProcess
  pixelDepthOffsetMode=PDOM_Legacy
  stencilRefValue=1
  usedShadingModels=Default Lit
 }
 nodes [
  Constant() {r=1}
  n1 = MaterialFunctionCall() {materialFunction=&/MF_WorldtoUV}
  n2 = MaterialFunctionCall(UVs=@n1.Result, "Rotation Angle (0-1)"=Constant() {r=0.25}) {materialFunction=&CustomRotator}
  n3 = ScalarParameter() {defaultValue=1, parameterName=GridScale}
  n4 = Divide(A=CollectionParameter() {collection=&/MPC_heatmap, parameterName=Gridsize}, B=@n3) {constB=3}
  n5 = TextureSample(UVs=Add(A=Multiply(A=@n2.Rotated Values, B=@n4) {constB=32}, B=Constant() {r=0.5})) {texture=&/T_edge}
  n6 = Multiply(A=@n5.R) {constB=0.3}
  n7 = Multiply(
   A=@n2.Rotated Values
   B=Divide(A=CollectionParameter() {collection=&MPC_heatmap, parameterName=Gridsize}, B=@n3) {constB=3}
  ) {constB=32}
  n8 = CollectionParameter() {collection=&MPC_heatmap, parameterName=MInnimapUvScale}
  n9 = Divide(B=CollectionParameter() {collection=&MPC_heatmap, parameterName=Gridsize}) {constB=24}
  LinearInterpolate(
   A=@n6
   B=Multiply(A=TextureSample(UVs=@n7) {texture=&/T_CreativeMoveTool_Grid}.G) {constB=0.15}
   Alpha=SphereMask(
    A=MaterialFunctionCall(UVs=@n1.Result, Texture Scale=@n8) {materialFunction=&ScaleUVsByCenter}.UVs
    B=MaterialFunctionCall!(cell) {materialFunction=&/MF_minimapUIUV}.UV_AimTargetGroupPos
    Radius=Multiply(A=@n9) {constB=2.12}
    Hardness=Constant() {r=1}
   )
  )
  n10 = Add(
   A=ScalarParameter() {defaultValue=7, parameterName=Test}
   B=ComponentMask(None=SceneTexture() {bFiltered=true, sceneTextureId=PPI_SceneDepth}.Color) {r=true}
  ) {constB=0.5}
  n11 = ComponentMask(None=SceneTexture() {bFiltered=true, sceneTextureId=PPI_CustomDepth}.Color) {r=true}
  n12 = ComponentMask(None=SceneTexture() {sceneTextureId=PPI_PostProcessInput0}.Color) {b=true, g=true, r=true}
  n13 = WorldPosition()
  n14 = CollectionParameter() {collection=&MPC_heatmap, parameterName=radius}
  n15 = Subtract(
   A=SphereMask(A=@n13.XYZ, B=CameraPositionWS(), Radius=Add(A=@n14) {constB=250}, Hardness=Constant() {r=0.98})
   B=SphereMask(A=@n13.XYZ, B=CameraPositionWS(), Radius=Add(A=@n14) {constB=250}, Hardness=Constant() {r=0.7})
  )
  n16 = Multiply(A=Saturate(None=Power(Base=@n15) {constExponent=5})) {constB=0.2}
  n17 = WorldPosition()
  n18 = CollectionParameter() {collection=&MPC_heatmap, parameterName=radius}
  n19 = SphereMask(A=@n17.XYZ, B=CameraPositionWS(), Radius=Add(A=@n18) {constB=250}, Hardness=Constant() {r=0.98})
  n20 = SphereMask(A=@n17.XYZ, B=CameraPositionWS(), Radius=Add(A=@n18) {constB=250}, Hardness=Constant() {r=0.7})
  n21 = Saturate(None=Power(Base=Subtract(A=@n19, B=@n20)) {constExponent=8})
  n22 = Multiply(A=MaterialFunctionCall() {materialFunction=&MF_WorldtoUV}.Result) {constB=1024}
  n23 = WorldPosition()
  n24 = CollectionParameter() {collection=&MPC_heatmap, parameterName=radius}
  n25 = Subtract(
   A=SphereMask(A=@n23.XYZ, B=CameraPositionWS(), Radius=Add(A=@n24) {constB=250}, Hardness=Constant() {r=0.98})
   B=SphereMask(A=@n23.XYZ, B=CameraPositionWS(), Radius=Add(A=@n24) {constB=230}, Hardness=Constant() {r=0.5})
  )
  n26 = Multiply(
   A=Multiply(A=TextureSample(UVs=@n22) {texture=/Game/Weapons/Prototype/ObjectMover/Textures/T_CreativeMoveTool_Grid}.R) {constB=0.5}
   B=Saturate(None=Power(Base=@n25) {constExponent=4})
  ) {constA=8, constB=2}
  n27 = TextureSample(UVs=Multiply(A=MaterialFunctionCall() {materialFunction=&MF_WorldtoUV}.Result) {constB=128}) {samplerType=SAMPLERTYPE_Masks, texture=/Game/Athena/Playlists/Barrier/Textures/T_BarrierGrid_M}
  n28 = WorldPosition()
  n29 = CollectionParameter() {collection=&MPC_heatmap, parameterName=radius}
  n30 = Subtract(
   A=SphereMask(A=@n28.XYZ, B=CameraPositionWS(), Radius=Add(A=@n29) {constB=250}, Hardness=Constant() {r=0.98})
   B=SphereMask(A=@n28.XYZ, B=CameraPositionWS(), Radius=Add(A=@n29) {constB=250}, Hardness=Constant() {r=0.7})
  )
  n31 = Time() {bOverride_Period=true, period=3.85}
  n32 = ScalarParameter() {defaultValue=1.5, parameterName=Sine Speed Grid}
  n33 = Sine(None=Add(A=@n27.B, B=Multiply(A=@n31, B=@n32)))
  n34 = Saturate(None=Power(Base=Multiply(A=@n27.G, B=@n33), Exp=ScalarParameter() {defaultValue=2, parameterName=Grid Power}))
  n35 = Saturate(None=@n25)
  n36 = Add(
   A=Multiply(A=Multiply(A=@n27.G) {constB=0.25}, B=Multiply(A=@n30) {constB=0.2}) {constB=0.5}
   B=Multiply(A=Multiply(A=@n34) {constB=0.25}, B=Multiply(A=@n35) {constB=0.1})
  )
  n37 = Add(A=@n16, B=Add(A=Multiply(A=@n21) {constB=0.25}, B=Add(A=@n26, B=@n36)))
  n38 = Saturate(None=QualitySwitch(Default=@n37, Low=@n16, High=@n37, Medium=@n16))
  n39 = CollectionParameter() {collection=&MPC_heatmap, parameterName=radius}
  n40 = Saturate(None=Multiply(A=Divide(A=@n39) {constB=100000}) {constB=20})
  n41 = SphereMask(A=WorldPosition().XYZ, B=CameraPositionWS(), Radius=Constant() {r=512}, Hardness=Constant() {r=0.5})
  n42 = TextureSample(UVs=MaterialFunctionCall() {materialFunction=&MF_WorldtoUV}.Result) {
   samplerSource=SSM_Clamp_WorldGroupSettings
   samplerType=SAMPLERTYPE_LinearColor
   texture=&/RT_Heatmapcode
  }
  n43 = TextureSample(UVs=MaterialFunctionCall() {materialFunction=&MF_WorldtoUV}.Result) {
   samplerSource=SSM_Clamp_WorldGroupSettings
   samplerType=SAMPLERTYPE_LinearColor
   texture=&RT_Heatmapcode
  }
  n44 = TextureSample(UVs=MaterialFunctionCall() {materialFunction=&MF_WorldtoUV}.Result) {
   samplerSource=SSM_Clamp_WorldGroupSettings
   samplerType=SAMPLERTYPE_LinearColor
   texture=&RT_Heatmapcode
  }
  n45 = Multiply(A=@n44.R) {constB=10}
  n46 = Multiply(A=MaterialFunctionCall() {materialFunction=&MF_WorldtoUV}.Result) {constB=32}
  n47 = Sine(None=Time() {bOverride_Period=true, period=3})
  n48 = Multiply(
   A=Multiply(A=Saturate(None=@n45), B=TextureSample(UVs=@n46) {texture=&/T_Exclamaition}.R)
   B=LinearInterpolate(Alpha=ConstantBiasScale(None=@n47)) {constA=0.25}
  )
  n49 = TextureSample(UVs=MaterialFunctionCall() {materialFunction=&MF_WorldtoUV}.Result) {
   samplerSource=SSM_Clamp_WorldGroupSettings
   samplerType=SAMPLERTYPE_LinearColor
   texture=&RT_Heatmapcode
  }
  n50 = TextureSample(UVs=Multiply(A=MaterialFunctionCall() {materialFunction=&MF_WorldtoUV}.Result) {constB=32}) {texture=&/T_LockscreenLocked}
  n51 = ConstantBiasScale(None=Sine(None=Time() {bOverride_Period=true, period=3}))
  n52 = TextureSample(UVs=MaterialFunctionCall() {materialFunction=&MF_WorldtoUV}.Result) {
   samplerSource=SSM_Clamp_WorldGroupSettings
   samplerType=SAMPLERTYPE_LinearColor
   texture=&RT_Heatmapcode
  }
  n53 = LinearInterpolate(
   A=If(A=@n43.R, B=Constant() {r=0.9}, "A > B"=@n48, "A == B"=@n48, "A < B"=Constant())
   B=Multiply(A=Multiply(A=@n49.G, B=@n50.R), B=LinearInterpolate(Alpha=@n51) {constA=0.25})
   Alpha=@n52.G
  )
  n54 = Add(
   A=MaterialFunctionCall(
    1=Constant3Vector() {constant=<rgba 0.0322 0.0206 1 0>}
    2=Constant3Vector() {constant=<rgba 0.0625 1 0.0625 1>}
    3=Constant3Vector() {constant=<rgba 1 0.8252 0 0>}
    4=Constant3Vector() {constant=<rgba 1 0 0 0>}
    A=@n42.R
   ) {materialFunction=&Lerp_Multiple_Float3}.Lerp 4 Inputs
   B=ShadingPathSwitch(Default=@n53, Deferred=@n53, Mobile=Constant()) {desc=Migrated from FeatureLevelSwitch}
  )
  n55 = MaterialFunctionCall(
   UVs=MaterialFunctionCall() {materialFunction=&MF_WorldtoUV}.Result
   Texture Scale=CollectionParameter() {collection=&MPC_heatmap, parameterName=MInnimapUvScale}
  ) {materialFunction=&ScaleUVsByCenter}
  n56 = MaterialFunctionCall() {materialFunction=&MF_minimapUIUV}
  n57 = CollectionParameter() {collection=&MPC_heatmap, parameterName=GroupRadius}
  n58 = Add(A=Multiply(A=Multiply(A=@n57) {constB=2}) {constB=0.5}) {constB=-0.5}
  n59 = Multiply(A=Divide(B=CollectionParameter() {collection=&MPC_heatmap, parameterName=Gridsize}), B=@n58) {constB=3}
  n60 = Constant() {r=0.85}
  n61 = LinearInterpolate(
   A=@n54
   B=Multiply(A=@n54, B=SphereMask(A=@n55.UVs, B=@n56.UV_AimTaregtPOS, Radius=@n59, Hardness=@n60))
   Alpha=CollectionParameter() {collection=&MPC_heatmap, parameterName=ThrowItems}
  )
  n62 = Multiply(A=@n61, B=Multiply(A=Multiply(A=@n38, B=@n40), B=OneMinus(None=@n41)))
  n63 = Saturate(None=Divide(A=CollectionParameter() {collection=&MPC_heatmap, parameterName=radius}) {constB=100000})
  n64 = MaterialFunctionCall(Float2=ComponentMask(None=WorldPosition().XYZ) {g=true, r=true}) {materialFunction=&BreakOutFloat2Components}
  n65 = ScalarParameter() {defaultValue=10000, parameterName=Dividesize2}
  n66 = AppendVector(A=Divide(A=@n64.R, B=@n65), B=Divide(A=@n64.G, B=@n65))
  n67 = Power(Base=TextureSample(UVs=Panner(Coordinate=@n66) {speedX=0.2, speedY=0.2}) {samplerType=SAMPLERTYPE_LinearColor, texture=&HighContrastNoise}.R)
  n68 = CollectionParameter() {collection=&MPC_heatmap, parameterName=DynamicSwitch}
  n69 = ScalarParameter() {defaultValue=1, parameterName=GridScale}
  n70 = MaterialFunctionCall(
   UVs=MaterialFunctionCall(
    UVs=@n1.Result
    Texture Scale=CollectionParameter() {collection=&MPC_heatmap, parameterName=MInnimapUvScale}
   ) {materialFunction=&ScaleUVsByCenter}.UVs
   Texture Scale=Divide(A=@n69, B=CollectionParameter() {collection=&MPC_heatmap, parameterName=Gridsize}) {constA=3, constB=32}
  ) {materialFunction=&ScaleUVsByCenter}
  n71 = Divide(
   A=Multiply(A=CollectionParameter() {collection=&MPC_heatmap, parameterName=Gridsize}) {constB=0.5}
   B=@n69
  ) {constA=16, constB=3}
  n72 = MaterialFunctionCall() {materialFunction=&MF_minimapUIUV}
  n73 = Divide(A=CollectionParameter() {collection=&MPC_heatmap, parameterName=Gridsize}, B=@n69) {constA=32, constB=3}
  n74 = MaterialFunctionCall(cell=Subtract(A=Add(A=@n70.UVs, B=@n71) {constB=3}, B=Multiply(A=@n72.UV_B2_playerLoctaion, B=@n73) {constB=3})) {materialFunction=&MF_minimapUIUV}
  n75 = LinearInterpolate(A=@n5.R, B=@n74.MASK_DynamicGroup3DGrid, Alpha=@n68)
  n76 = Multiply(A=@n75) {constB=6}
  n77 = Add(A=Multiply(A=@n75) {constB=2}, B=@n75)
  n78 = Add(A=Multiply(A=@n67, B=Power(Base=@n76) {constExponent=10}), B=Add(A=Multiply(A=@n67, B=@n77), B=Multiply(A=@n75) {constB=0.25}))
  n79 = Constant3Vector() {constant=<rgba 0.3658 0 1 0>}
  n80 = Constant3Vector() {constant=<rgba 0 3.6243 9 0>}
  n81 = CollectionParameter() {collection=&MPC_heatmap, parameterName=Gridsize}
  n82 = TextureSample(UVs=Add(A=Multiply(A=@n81, B=@n2.Rotated Values) {constB=32}, B=Constant() {r=0.5})) {texture=&T_edge}
  n83 = Add(
   A=Add(A=LinearInterpolate(B=@n6, Alpha=@n68), B=Multiply(A=@n82.R) {constB=0.25})
   B=Multiply(A=@n78, B=LinearInterpolate(A=@n79, B=@n80, Alpha=@n78))
  )
  n84 = SphereMask(A=WorldPosition().XYZ, B=CameraPositionWS(), Radius=Constant() {r=50000}, Hardness=Constant() {r=0.5})
  n85 = WorldPosition()
  n86 = CameraPositionWS()
  n87 = Add(A=CollectionParameter() {collection=&MPC_heatmap, parameterName=radius}) {constB=250}
  n88 = Constant() {r=0.98}
  n89 = LinearInterpolate(
   A=Constant()
   B=Multiply(A=Multiply(A=@n83, B=@n84), B=SphereMask(A=@n85.XYZ, B=@n86, Radius=@n87, Hardness=@n88))
   Alpha=CollectionParameter() {collection=&MPC_heatmap, parameterName=spatialswitch}
  )
  n90 = ComponentMask(None=SceneTexture() {sceneTextureId=PPI_PostProcessInput0}.Color) {b=true, g=true, r=true}
  n91 = Constant() {r=0.5}
  n92 = WorldPosition()
  n93 = CollectionParameter() {collection=&MPC_heatmap, parameterName=radius}
  n94 = Subtract(
   A=SphereMask(A=@n92.XYZ, B=CameraPositionWS(), Radius=Add(A=@n93) {constB=250}, Hardness=Constant() {r=0.98})
   B=SphereMask(A=@n92.XYZ, B=CameraPositionWS(), Radius=@n93, Hardness=Constant() {r=0.8})
  )
  n95 = WorldPosition()
  n96 = CameraPositionWS()
  n97 = Constant() {r=256}
  n98 = Constant() {r=0.5}
  n99 = MaterialFunctionCall() {materialFunction=&MF_WorldtoUV}
  n100 = Multiply(
   A=OneMinus(None=SphereMask(A=@n95.XYZ, B=@n96, Radius=@n97, Hardness=@n98))
   B=TextureSample(UVs=Multiply(A=@n99.Result) {constB=256}) {texture=&T_CreativeMoveTool_Grid}.R
  )
  n101 = Constant() {r=1}
  n102 = Multiply(A=Multiply(A=Multiply(A=@n61) {constB=2}, B=@n101), B=@n101)
  n103 = Multiply(A=Multiply(A=Clamp(None=@n94), B=Multiply(A=@n100) {constA=8, constB=2}), B=Multiply(B=@n102) {constA=8})
  n104 = Divide(A=CollectionParameter() {collection=&MPC_heatmap, parameterName=radius}) {constB=100000}
  n105 = TextureSample(UVs=Multiply(A=MaterialFunctionCall() {materialFunction=&MF_WorldtoUV}.Result) {constB=1024}) {texture=&T_CreativeMoveTool_Grid}
  n106 = Multiply(A=@n105.R) {constB=0.5}
  n107 = SceneTexture() {sceneTextureId=PPI_WorldNormal}
  n108 = Multiply(B=Fresnel(ExponentIn=Constant() {r=1}, Normal=ComponentMask(None=@n107.Color) {b=true, g=true, r=true})) {constA=8, constB=3}
  n109 = Fresnel(ExponentIn=Constant() {r=6}, Normal=ComponentMask(None=SceneTexture() {sceneTextureId=PPI_WorldNormal}.Color) {b=true, g=true, r=true})
  n110 = Add(A=Add(A=Multiply(A=@n106, B=@n108), B=Multiply(B=@n109) {constA=1.3, constB=3}) {constB=0.5}) {constB=0.3}
  n111 = Add(A=@n105.R) {constB=0.3}
  n112 = ShadingPathSwitch(Default=@n110, Deferred=@n110, Forward=@n111, Mobile=@n111)
  n113 = Add(
   A=LinearInterpolate(
    A=MaterialFunctionCall() {materialFunction=&/MF_StaticGroup3DPost}.Result
    B=MaterialFunctionCall() {materialFunction=&/MF_DynamicGroup3DPost}.Result
    Alpha=CollectionParameter() {collection=&MPC_heatmap, parameterName=DynamicSwitch}
   )
   B=Add(A=LinearInterpolate(A=@n103, Alpha=@n104) {constA=1, constB=0}, B=Multiply(A=@n102, B=@n112) {constB=2})
  )
  n114 = Desaturation(None=ComponentMask(None=SceneTexture() {sceneTextureId=PPI_PostProcessInput0}.Color) {b=true, g=true, r=true})
  n115 = Desaturation(None=ComponentMask(None=SceneTexture() {sceneTextureId=PPI_PostProcessInput0}.Color) {b=true, g=true, r=true})
  n116 = LinearInterpolate(
   A=Multiply(A=Desaturation(None=@n90), B=@n91) {constB=0.5}
   B=Add(A=Multiply(A=@n113, B=@n114), B=Multiply(A=@n115, B=@n91) {constB=0.5})
   Alpha=SphereMask(
    A=WorldPosition().XYZ
    B=CameraPositionWS()
    Radius=CollectionParameter() {collection=&MPC_heatmap, parameterName=radius}
    Hardness=Constant() {r=0.9}
   )
  )
  n117 = CollectionParameter() {collection=&MPC_heatmap, parameterName=postSwitch}
  n118 = Constant()
  n119 = CollectionParameter() {collection=&MPC_heatmap, parameterName=ThrowItems}
  n120 = LinearInterpolate(
   A=LinearInterpolate(A=Add(A=@n89, B=@n116), B=@n12, Alpha=LinearInterpolate(A=@n117, B=@n118, Alpha=@n119))
   B=@n12
   Alpha=Constant() {r=0.15}
  )
  n121 = CollectionParameter() {collection=&MPC_heatmap, parameterName=radius}
  n122 = Saturate(None=Multiply(A=Divide(A=@n121) {constB=100000}) {constB=3})
  LinearInterpolate(
   A=Add(A=LinearInterpolate(A=@n62, Alpha=@n63) {constA=1, constB=0}, B=LinearInterpolate(A=@n12, B=@n120, Alpha=@n122))
   B=@n12
   Alpha=Clamp(None=Subtract(A=@n10, B=@n11))
  )
  Multiply(B=Subtract(A=@n58) {constB=0.01}) {constB=2.225}
  Multiply(A=@n5.R) {constB=0.5}
  OneMinus()
  n123 = MaterialFunctionCall() {materialFunction=&MF_WorldtoUV}
  n124 = CollectionParameter() {collection=&MPC_heatmap, parameterName=MInnimapUvScale}
  n125 = Divide(B=CollectionParameter() {collection=&MPC_heatmap, parameterName=Gridsize})
  n126 = OneMinus(None=SphereMask(
    A=MaterialFunctionCall(UVs=@n123.Result, Texture Scale=@n124) {materialFunction=&ScaleUVsByCenter}.UVs
    B=MaterialFunctionCall() {materialFunction=&MF_minimapUIUV}.UV_B2_playerLoctaion
    Radius=Multiply(A=@n125) {constB=2.12}
    Hardness=Constant() {r=1}
   ))
  Reroute(None=Multiply(A=Multiply(A=@n74.MASK_DynamicGroup3DGrid, B=@n126), B=Constant3Vector() {constant=<rgba 0.0086 0.0603 0.15 0>}))
  TextureCoordinate()
  TextureCoordinate()
 ]
}
