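# Creative performance HUD widget that displays runtime stats in the Playground thermometer, such as frame, thread and GPU times, draw calls and free memory.
# SetRuntimeStats stores the incoming text array, removes index 3 (the GPU time slot) on mobile, then fills eight text blocks by index and blanks the unused ones.
# Construct builds that text block array and positions the scale box on the canvas with separate mobile and default offsets.
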WidgetBlueprint /Game/Athena/HUD/LTM/Playground/Thermometer/ThermometeRuntime
graphs [
 EventGraph {nodes [
   n1 = Event SetRuntimeStats!() -> (OutputDelegate, then, DetailRuntimeStats:Array of Texts) {is=title K2Node_Event}
   > Set RunTimeStatsText(RunTimeStatsText:Array of Texts=@n1.DetailRuntimeStats) -> (Output_Get:Array of Texts)
   > n2 = Branch(Condition=NOT Boolean(A=IsMobileGame()))
   n3 = For Each Loop(Exec=[@n6.then, @n2.then], Array:Array of Texts=Get RunTimeStatsText() -> (RunTimeStatsText:Array of Texts)) -> (Array Element:Text)
   n4 = Get RuntimeTextWidgets() -> (RuntimeTextWidgets:Array of Common Text Object References)
   n5 = For Loop(
    execute=@n3.Completed
    FirstIndex=Length(TargetArray:Array of Texts=Get RunTimeStatsText())
    LastIndex=Last Index!(TargetArray:Array of Common Text Object References=@n4) -> (ReturnValue=0)
   )
   > "SetText (Text)"(self="Get (a copy)"(Array:Array of Common Text Object References=@n4, Dimension 1=@n5) -> (Output:Common Text Object Reference))
   "SetText (Text)"(execute=@n3.LoopBody, self="Get (a copy)"(Array=Get RuntimeTextWidgets(), Dimension 1=@n3.Array Index), InText=@n3)
   n6 = Remove Index!(execute=@n2.else, TargetArray:Array of Texts=Get RunTimeStatsText(), IndexToRemove:int=3) -> (then)
   Event Construct()
   > Set RuntimeTextWidgets(RuntimeTextWidgets:Array of Common Text Object References=Make Array!(
     "[0]":Common Text Object Reference=Get 0() -> (0:Common Text Object Reference)
     "[1]":Common Text Object Reference=Get 1() -> (1:Common Text Object Reference)
     "[2]":Common Text Object Reference=Get 2() -> (2:Common Text Object Reference)
     "[3]":Common Text Object Reference=Get 3() -> (3:Common Text Object Reference)
     "[4]":Common Text Object Reference=Get 4() -> (4:Common Text Object Reference)
     "[5]":Common Text Object Reference=Get 5() -> (5:Common Text Object Reference)
     "[6]":Common Text Object Reference=Get 6() -> (6:Common Text Object Reference)
     "[7]":Common Text Object Reference=Get 7() -> (7:Common Text Object Reference)
    ) -> (Array:Array of Common Text Object References)) -> (Output_Get:Array of Common Text Object References)
   > SetWidgetPositionOnPlatforms!(
    execute
    self
    isApplyToBattleLab:bool=false
    MobilePos:Vector 2D Structure=(X=-520,Y=36)
    DefaultPos:Vector 2D Structure=(X=-480,Y=36)
    WidgetToRepos:Widget Object Reference=Get ThermometerArea_Runtime() -> (ThermometerArea_Runtime:Scale Box Object Reference)
   ) -> (then)
  ]}
 SetWidgetPositionOnPlatforms {nodes [
   n1 = SetWidgetPositionOnPlatforms!() -> (then, isApplyToBattleLab:bool=~, MobilePos:Vector 2D Structure, DefaultPos:Vector 2D Structure, WidgetToRepos:Widget Object Reference) {is=title K2Node_FunctionEntry}
   > n2 = Branch(Condition=@n1)
   > n3 = Cast To FortGameStateAthena!(execute, Object=GetGameState!() -> (ReturnValue:Game State Base Object Reference)) -> (then, CastFailed, AsFort Game State Athena:Fort Game State Athena Object Reference) {is=title K2Node_DynamicCast}
   > Is Valid(InputObject=@n3)
   > n4 = Branch(Condition=IsMobileGame())
   n5 = Set TargetPos(execute=[@n4.then, @n8.then], TargetPos:Vector 2D Structure=@n1.MobilePos) -> (Output_Get:Vector 2D Structure)
   n6 = SlotAsCanvasSlot!(Widget:Widget Object Reference=@n1.WidgetToRepos) -> (ReturnValue:Canvas Panel Slot Object Reference)
   Is Valid(exec=[@n7.then, @n5.then], InputObject=@n6)
   > SetPosition!(execute, self:Canvas Panel Slot Object Reference=@n6, InPosition:Vector 2D Structure=Get TargetPos() -> (TargetPos:Vector 2D Structure)) -> (then)
   > Return Node()
   n7 = Set TargetPos(execute=[@n8.else, @n4.else], TargetPos=@n1.DefaultPos)
   n8 = Branch(execute=@n2.else, Condition=IsMobileGame())
  ]}
]
variables [
 RunTimeStatsText {type="array<string>", default []}
 RuntimeTextWidgets {type="array<CommonTextBlock>", default []}
]
widgets {
 info {
  inheritedWidgetCount=1
  parentClass=/Script/FortniteUI.CreativePerformanceHUDRunTimeStatsWidget
  rootWidgetClass=&CanvasPanel
  widgetCount=16
 }
 widgets [
  CanvasPanel_0 {widgetClassPath=&CanvasPanel, widget=CanvasPanel_0}
  ThermometerArea_Runtime {
   widgetClassPath=&ScaleBox
   bIsVariable=true
   parent=CanvasPanel_0
   widget=ThermometerArea_Runtime
   properties {stretch=UserSpecified, userSpecifiedScale=0.8, visibility=HitTestInvisible}
   slot {
    class=CanvasPanelSlot
    properties {bAutoSize=true, layoutData {alignment=<1 0>, anchors {maximum=<1 0>, minimum=<1 0>}, offsets=<ltrb -520 36 100 30>}}
   }
  }
  VerticalBox_2 {widgetClassPath=&VerticalBox, bIsVariable=true, parent=ThermometerArea_Runtime, widget=VerticalBox_2, slot {class=ScaleBoxSlot}}
  RUNTime {
   widgetClassPath=&CommonTextBlock
   bIsVariable=true
   parent=VerticalBox_2
   widget=RUNTime
   properties {
    text=Runtime
    visibility=Collapsed
    font {
     bForceMonospaced=false
     bMaterialIsStencil=false
     fontMaterial=null
     fontObject=&NotoSans
     letterSpacing=0
     monospacedWidth=1
     size=11
     skewAmount=0
     typefaceFontName=NotoSans-Bold
     outlineSettings {
      bApplyOutlineToDropShadows=false
      bMiteredCorners=false
      bSeparateFillAlpha=false
      outlineMaterial=null
      outlineSize=0
      outlineColor=<rgba 0 0 0 1>
     }
    }
    shadowOffset=<0 0>
   }
   slot {class=VerticalBoxSlot, properties {horizontalAlignment=HAlign_Left}}
  }
  VerticalBox_0 {
   widgetClassPath=&VerticalBox
   parent=VerticalBox_2
   widget=VerticalBox_0
   slot {class=VerticalBoxSlot, properties {horizontalAlignment=HAlign_Left}}
  }
  mem {widgetClassPath=&HorizontalBox, parent=VerticalBox_0, widget=mem, properties {visibility=Visible}, slot {class=VerticalBoxSlot}}
  VerticalBox_1 {
   widgetClassPath=&VerticalBox
   parent=mem
   widget=VerticalBox_1
   slot {class=HorizontalBoxSlot, properties {padding=<ltrb 2 2 2 2>}}
  }
  0 {
   widgetClassPath=&CommonTextBlock
   bIsVariable=true
   parent=VerticalBox_1
   widget=0
   properties {
    text=FrameTime
    font {
     bForceMonospaced=false
     bMaterialIsStencil=false
     fontMaterial=null
     fontObject=&NotoSans
     letterSpacing=0
     monospacedWidth=1
     size=11
     skewAmount=0
     typefaceFontName=NotoSans-Bold
     outlineSettings {
      bApplyOutlineToDropShadows=false
      bMiteredCorners=false
      bSeparateFillAlpha=false
      outlineMaterial=null
      outlineSize=0
      outlineColor=<rgba 0 0 0 1>
     }
    }
    shadowOffset=<0 0>
   }
   slot {class=VerticalBoxSlot}
  }
  1 {
   widgetClassPath=&CommonTextBlock
   bIsVariable=true
   parent=VerticalBox_1
   widget=1
   properties {
    text=GameThreadTime
    font {
     bForceMonospaced=false
     bMaterialIsStencil=false
     fontMaterial=null
     fontObject=&NotoSans
     letterSpacing=0
     monospacedWidth=1
     size=11
     skewAmount=0
     typefaceFontName=NotoSans-Bold
     outlineSettings {
      bApplyOutlineToDropShadows=false
      bMiteredCorners=false
      bSeparateFillAlpha=false
      outlineMaterial=null
      outlineSize=0
      outlineColor=<rgba 0 0 0 1>
     }
    }
    shadowOffset=<0 0>
   }
   slot {class=VerticalBoxSlot}
  }
  2 {
   widgetClassPath=&CommonTextBlock
   bIsVariable=true
   parent=VerticalBox_1
   widget=2
   properties {
    text=RenderThreadTime
    font {
     bForceMonospaced=false
     bMaterialIsStencil=false
     fontMaterial=null
     fontObject=&NotoSans
     letterSpacing=0
     monospacedWidth=1
     size=11
     skewAmount=0
     typefaceFontName=NotoSans-Bold
     outlineSettings {
      bApplyOutlineToDropShadows=false
      bMiteredCorners=false
      bSeparateFillAlpha=false
      outlineMaterial=null
      outlineSize=0
      outlineColor=<rgba 0 0 0 1>
     }
    }
    shadowOffset=<0 0>
   }
   slot {class=VerticalBoxSlot}
  }
  3 {
   widgetClassPath=&CommonTextBlock
   bIsVariable=true
   parent=VerticalBox_1
   widget=3
   properties {
    text=TotalGPUTime
    font {
     bForceMonospaced=false
     bMaterialIsStencil=false
     fontMaterial=null
     fontObject=&NotoSans
     letterSpacing=0
     monospacedWidth=1
     size=11
     skewAmount=0
     typefaceFontName=NotoSans-Bold
     outlineSettings {
      bApplyOutlineToDropShadows=false
      bMiteredCorners=false
      bSeparateFillAlpha=false
      outlineMaterial=null
      outlineSize=0
      outlineColor=<rgba 0 0 0 1>
     }
    }
    shadowOffset=<0 0>
   }
   slot {class=VerticalBoxSlot}
  }
  4 {
   widgetClassPath=&CommonTextBlock
   bIsVariable=true
   parent=VerticalBox_1
   widget=4
   properties {
    text=RhIThreadTime
    font {
     bForceMonospaced=false
     bMaterialIsStencil=false
     fontMaterial=null
     fontObject=&NotoSans
     letterSpacing=0
     monospacedWidth=1
     size=11
     skewAmount=0
     typefaceFontName=NotoSans-Bold
     outlineSettings {
      bApplyOutlineToDropShadows=false
      bMiteredCorners=false
      bSeparateFillAlpha=false
      outlineMaterial=null
      outlineSize=0
      outlineColor=<rgba 0 0 0 1>
     }
    }
    shadowOffset=<0 0>
   }
   slot {class=VerticalBoxSlot}
  }
  5 {
   widgetClassPath=&CommonTextBlock
   bIsVariable=true
   parent=VerticalBox_1
   widget=5
   properties {
    text=drawcall
    font {
     bForceMonospaced=false
     bMaterialIsStencil=false
     fontMaterial=null
     fontObject=&NotoSans
     letterSpacing=0
     monospacedWidth=1
     size=11
     skewAmount=0
     typefaceFontName=NotoSans-Bold
     outlineSettings {
      bApplyOutlineToDropShadows=false
      bMiteredCorners=false
      bSeparateFillAlpha=false
      outlineMaterial=null
      outlineSize=0
      outlineColor=<rgba 0 0 0 1>
     }
    }
    shadowOffset=<0 0>
   }
   slot {class=VerticalBoxSlot}
  }
  6 {
   widgetClassPath=&CommonTextBlock
   bIsVariable=true
   parent=VerticalBox_1
   widget=6
   properties {
    text=primitives drawn
    font {
     bForceMonospaced=false
     bMaterialIsStencil=false
     fontMaterial=null
     fontObject=&NotoSans
     letterSpacing=0
     monospacedWidth=1
     size=11
     skewAmount=0
     typefaceFontName=NotoSans-Bold
     outlineSettings {
      bApplyOutlineToDropShadows=false
      bMiteredCorners=false
      bSeparateFillAlpha=false
      outlineMaterial=null
      outlineSize=0
      outlineColor=<rgba 0 0 0 1>
     }
    }
    shadowOffset=<0 0>
   }
   slot {class=VerticalBoxSlot}
  }
  7 {
   widgetClassPath=&CommonTextBlock
   bIsVariable=true
   parent=VerticalBox_1
   widget=7
   properties {
    text=raw physical memory free
    font {
     bForceMonospaced=false
     bMaterialIsStencil=false
     fontMaterial=null
     fontObject=&NotoSans
     letterSpacing=0
     monospacedWidth=1
     size=11
     skewAmount=0
     typefaceFontName=NotoSans-Bold
     outlineSettings {
      bApplyOutlineToDropShadows=false
      bMiteredCorners=false
      bSeparateFillAlpha=false
      outlineMaterial=null
      outlineSize=0
      outlineColor=<rgba 0 0 0 1>
     }
    }
    shadowOffset=<0 0>
   }
   slot {class=VerticalBoxSlot}
  }
  TutorialHighlight {widgetClassPath=&TutorialHighlight}
 ]
}
