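; Playground thermometer actor that drives the 3D heatmap post-process effect, clearing its render target on BeginPlay and binding play and stop events.
; PlayFX enables post-processing, runs timelines animating material collection radius, postSwitch and chromatic settings, and plays a placeholder sound; StopFX disables it.
; SpawnBox snaps player XY to a 32x32 grid cell, moving and scaling a cube there, with debug printing.
; Construction creates two heatmap UI dynamic materials.
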(event EventBeginPlay
  (Rendering|ClearRenderTarget2D "/Game/Athena/HUD/LTM/Playground/Thermometer/RT_Heatmapcode.RT_Heatmapcode")
  (Default|BindEventtoPlaypostprocess (delegate Custom|PlayFX))
  (Default|BindEventtoDisablePostProcess (delegate Custom|StopFX)))

(event Custom|PlayFX
  (Variables|/Script/Engine.PostProcessComponent|SetbEnabled true (Variables|Default|GetPostProcess))
  (Utilities|FlowControl|Sequence
    (:then_0
      (Rendering|Material|SetScalarParameterValue "/Game/Athena/HUD/LTM/Playground/Thermometer/MPC_heatmap.MPC_heatmap" "animation Switch" 1.0 :!class /Script/Engine.KismetMaterialLibrary)
      (bind (_direction _newtrack_0) (Timeline|Timeline_0 PlayFromStart)
        (:Update
          (Rendering|Material|SetScalarParameterValue "/Game/Athena/HUD/LTM/Playground/Thermometer/MPC_heatmap.MPC_heatmap" "radius" _newtrack_0 :!class /Script/Engine.KismetMaterialLibrary))))
    (:then_1
      (bind (_direction_2 _postswitch) (Timeline|Timeline_2 PlayFromStart)
        (:Update
          (Rendering|Material|SetScalarParameterValue "/Game/Athena/HUD/LTM/Playground/Thermometer/MPC_heatmap.MPC_heatmap" "postSwitch" _postswitch :!class /Script/Engine.KismetMaterialLibrary))))
    (:then_2
      (bind (_direction_3 _chromatic) (Timeline|Timeline_3 PlayFromStart)
        (:Update
          (Variables|/Script/Engine.PostProcessComponent|SetSettings (Utilities|Struct|MakePostProcessSettings _chromatic (Utilities|Struct|MakeWeightedBlendables (Utilities|Array|MakeArray (Variables|Default|GetNewVar_5)))) (Variables|Default|GetPostProcess)))))
    (:then_3
      ;[1 0.6165 0] TODO: Replace with Proper Sound Cue
      ; Currently using MoveTool Ones as a Placeholder
      (Audio|PlaySound2D (Variables|Default|GetActivationSFX) 1.5))))

(event Custom|StopFX
  (Variables|/Script/Engine.PostProcessComponent|SetbEnabled false (Variables|Default|GetPostProcess)))

(fn SpawnBox
  (params (PlayerX :type Float%20%28double-precision%29) (PlayerY :type Float%20%28double-precision%29) (PlayerZ :type Float%20%28double-precision%29))
  (bind _cube (Variables|Default|GetCube))
  (bind _output_get (Variables|Default|SetboundaryLoc (Math|Vector|MakeVector (+ (+ (Variables|Default|GetminsizeX) (* (/ (- (Variables|Default|GetMaxsizeX) (Variables|Default|GetminsizeX)) 32.0) (Math|Conversions|ToFloat(Integer) (Math|Float|Floor (/ (- PlayerX (Variables|Default|GetminsizeX)) (/ (- (Variables|Default|GetMaxsizeX) (Variables|Default|GetminsizeX)) 32.0)))))) (* (/ (- (Variables|Default|GetMaxsizeX) (Variables|Default|GetminsizeX)) 32.0) 0.5)) (+ (+ (Variables|Default|GetminsizeY) (* (/ (- (Variables|Default|GetMaxsizeY) (Variables|Default|GetminsizeY)) 32.0) (Math|Conversions|ToFloat(Integer) (Math|Float|Floor (/ (- PlayerY (Variables|Default|GetminsizeY)) (/ (- (Variables|Default|GetMaxsizeY) (Variables|Default|GetminsizeY)) 32.0)))))) (* (/ (- (Variables|Default|GetMaxsizeY) (Variables|Default|GetminsizeY)) 32.0) 0.5)) (+ 0.0 0.0))))
  (Transformation|SetWorldScale3D _cube (Math|Vector|MakeVector (/ (Variables|Default|Getvolume%20size) 3.2e+03) (/ (Variables|Default|Getvolume%20size) 3.2e+03) 5e+02))
  (Transformation|SetWorldLocation _cube _output_get false false)
  (Development|PrintString (Utilities|String|ToString(Float) (Variables|Default|Getvolume%20size)) true true "(R=0.679543,G=0.027182,B=0.027182,A=1.000000)"))

(fn ConstructionScript
  (bind _returnvalue (Rendering|Material|CreateDynamicMaterialInstance "/Game/Athena/HUD/LTM/Playground/Thermometer/M_HeatmapUI.M_HeatmapUI" "" "None" :!class /Script/Engine.KismetMaterialLibrary))
  (Variables|Default|SetHeatMapUI _returnvalue)
  (bind _returnvalue_2 (Rendering|Material|CreateDynamicMaterialInstance "/Game/Athena/HUD/LTM/Playground/Thermometer/M_HeatmapUI32.M_HeatmapUI32" "" "None" :!class /Script/Engine.KismetMaterialLibrary))
  (Variables|Default|SetHeatmapUI32 _returnvalue_2))
